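# Additive unlit flame material for the birthday candle set, enabled for Niagara sprites and skeletal meshes.
# Panning cloud noise, weighted by inverted V so it fades along the sprite, distorts UVs of a clamped, tiled and offset flame texture.
# Color blends particle color toward texture color, then applies a tint.
# Opacity multiplies particle alpha by selected texture channels and a scale parameter.
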Material /DealershipSharedLibrary/Assets/Sets/BirthdayCandle/Materials/M_DLR_CandleFlame
graph {
 properties {
  bAutomaticallySetUsageInEditor=false
  bUsedWithNiagaraSprites=true
  bUsedWithSkeletalMesh=true
  blendMode=BLEND_Additive
  shadingModel=MSM_Unlit
  usedShadingModels=Unlit
 }
 nodes [
  MaterialFunctionCall!(Center Location, 0-1 UVs, XY scale, OverridePivot, "Normals (Optional)") {materialFunction=&Sprite}
  n1 = ScalarParameter() {defaultValue=1, group=Emissive Color, parameterName="Emissive - Use Particle Color (0) or Texture Color (1)?", sortPriority=1}
  n2 = TextureCoordinate()
  n3 = VectorParameter() {
   group=Distortion
   parameterName=Distortion - Noise Tiling
   channelNames {a=Speed Y, b=Speed X, g=Tiling Y, r=TIling X}
   defaultValue=<rgba 0.25 0.25 1 1>
  }
  n4 = AppendVector(A=@n3.TIling X, B=@n3.TIling X)
  n5 = TextureSampleParameter2D(UVs=Panner(Coordinate=Multiply(A=@n2, B=@n4), Time=Time(), Speed=AppendVector(A=@n3.Speed X, B=@n3.Speed X))) {group=Distortion, parameterName=Distortion - Noise, texture=/DealershipSharedLibrary/FX/Textures/Noise/T_VCVFX_CloudDistort_A}
  n6 = StaticComponentMaskParameter(None=TextureCoordinate()) {defaultG=true, group=Distortion, parameterName=Distortion - Noise Direction}
  n7 = ConstantBiasScale(None=Multiply(A=StaticComponentMaskParameter(None=@n5.R) {defaultR=true, group=Distortion, parameterName=DIstortion - Noise Channel}, B=OneMinus(None=@n6))) {bias=0, scale=0.75}
  n8 = ScalarParameter() {defaultValue=1, group=Distortion, parameterName=Distortion - Amount}
  n9 = VectorParameter() {
   group=Flame Texture
   parameterName=Flame Texture - Tiling
   sortPriority=3
   channelNames {a=Speed Y, b=Speed X, g=Tiling Y, r=TIling X}
   defaultValue=<rgba 1.75 1.75 -0.45 -0.7>
  }
  n10 = Multiply(A=TextureCoordinate(), B=AppendVector(A=@n9.TIling X, B=@n9.TIling X))
  n11 = AppendVector(A=@n9.Speed X, B=@n9.Speed X)
  n12 = TextureSampleParameter2D(UVs=Add(A=Multiply(A=@n7, B=@n8), B=Add(A=@n10, B=@n11))) {
   group=Flame Texture
   parameterName="Flame Texture (Clamped, Not Repeating)"
   samplerSource=SSM_Clamp_WorldGroupSettings
   sortPriority=3
   texture=/DealershipSharedLibrary/Assets/Sets/BirthdayCandle/Texture/T_DLR_CandleFlame
  }
  Multiply(
   A=LinearInterpolate(A=ParticleColor().RGB, B=@n12.RGB, Alpha=@n1)
   B=VectorParameter() {group=Emissive Color, parameterName=Emissive - Tint, sortPriority=1, defaultValue=<rgba 2 1 1 1>}.RGB
  )
  Multiply(
   A=Multiply(
    A=ParticleColor().A
    B=StaticComponentMaskParameter(None=@n12.RGBA) {defaultB=true, defaultG=true, defaultR=true, group=Opacity, parameterName=Opacity - Channel, sortPriority=2}
   )
   B=ScalarParameter() {defaultValue=1, group=Opacity, parameterName=Opacity - Scale, sortPriority=2}
  )
  NamedRerouteDeclaration() {name=WPO, nodeColor=<rgba 0.2187 1 0 1>}
  NamedRerouteUsage()
 ]
}
